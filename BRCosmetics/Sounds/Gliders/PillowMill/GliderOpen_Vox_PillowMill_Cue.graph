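# Sound cue for the opening of the PillowMill cosmetic glider, layering an open sound with an occasional vocal.
# A mixer plays one of three equally weighted open waves alongside a vocal branch whose volume is randomly lowered by a modulator.
# The vocal random node interleaves three empty slots with three Vox waves at equal weight, so a voice plays about half the time.
# Shared glider open/close attenuation and concurrency assets govern playback.
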SoundCue /BRCosmetics/Sounds/Gliders/PillowMill/GliderOpen_Vox_PillowMill_Cue
root=Mixer_0
nodes [
 {id=Mixer_0, children [Random_0,Modulator_0], properties {inputVolume [1,1]}}
 {id=Modulator_0, children [Random_1], properties {volumeMax=0.6, volumeMin=0.5}}
 {id=Random_0, children [WavePlayer_0,WavePlayer_1,WavePlayer_2], properties {weights [1,1,1]}}
 {id=Random_1, children [null,WavePlayer_3,null,WavePlayer_4,null,WavePlayer_5], properties {weights [1,1,1,1,1,1]}}
 {id=WavePlayer_0, properties {soundWaveAssetPtr=&/IP_Glider_PillowMill_Open_01}}
 {id=WavePlayer_1, properties {soundWaveAssetPtr=&/IP_Glider_PillowMill_Open_02}}
 {id=WavePlayer_2, properties {soundWaveAssetPtr=&/IP_Glider_PillowMill_Open_03}}
 {id=WavePlayer_3, properties {soundWaveAssetPtr=&/IP_Glider_PillowMill_Open_Vox_03}}
 {id=WavePlayer_4, properties {soundWaveAssetPtr=&/IP_Glider_PillowMill_Open_Vox_02}}
 {id=WavePlayer_5, properties {soundWaveAssetPtr=&/IP_Glider_PillowMill_Open_Vox_01}}
]
properties {
 attenuationSettings=/Game/Sounds/Attenuation/Fort_Player_Glider_OpenClose_Athena_Att
 audioPropertiesSheet=/BRCosmetics/Sounds/Settings/PropertySheets/Glider/APS_Cosmetic_Glider_Open
 maxDistance=10000
 soundClassObject=&Fort_Gliders_Land
 volumeMultiplier=0.8
 concurrencySet [&Athena_GliderOpenClose_Conc]
}
